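# Niagara effect for the CountyFrog Loaf emote that emits sprite bubbles from the emote's net skeletal mesh.
# One emitter fires a burst plus a brief early stream; bubbles inherit some mesh motion, swirl with curl noise and lift late in life.
# Alpha scales with time of day, and front-end modifiers change velocity and drag.
# In gameplay, a world/local position blend follows owner speed to handle traversal.
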NiagaraSystem /BRCosmetics/Animation/Game/MainPlayer/Emotes/CountyFrog_Loaf/FX/NS_Emote_CountyFrogLoaf_NetBubbles
dependencies {
 usedDataInterfaces [
  &NiagaraDataInterfaceCurve
  &NiagaraDataInterfaceVector2DCurve
  &NiagaraDataInterfaceSkeletalMesh
 ]
 usedDynamicInputs [&FNTimeOfDayScalar]
 usedModules [
  &self:SystemSpawnScript.SystemScriptSource.SystemScriptGraph.NiagaraNodeAssignment_0.SetVariables_#1
  &self:SystemSpawnScript.SystemScriptSource.SystemScriptGraph.NiagaraNodeAssignment_1.SetVariables_#2
  &SystemState
  &BlendSpaceHelper
 ]
}
emitters {BigJelly_Bubbles {
  emitterSpawnScript [/SetVariables(Emitter.IsTraversing Speed=100, Emitter.WalkForward Speed=300) {isSetParameters=true}]
  emitterUpdateScript [
   &EmitterState(Life Cycle Mode=::NewEnumerator1, Loop Behavior=::NewEnumerator1, Loop Duration=4, UseLoopDelay=false)
   &SpawnBurst_Instantaneous(Spawn Count=10, Use Spawn Probability=false, UseLoopCountLimit=false)
   &SpawnRate(
    SpawnRate=&Multiply_Float(
     A=&FloatFromCurve(FloatCurve=DataInterface() {propertyValues {Curve {keys [<at 0.1458 1>,<at 0.1787 0>]}}}, CurveIndex=$Emitter.NormalizedLoopAge)
     B=115
    )
    Use Spawn Probability=false
   )
  ]
  particleSpawnScript [
   &InitializeParticle(
    Lifetime Max=1.25
    Lifetime Min=0.2
    Lifetime Mode=::NewEnumerator1
    Mass Mode=::NewEnumerator3
    Sprite Rotation Mode=::NewEnumerator2
    Sprite Size Mode=::NewEnumerator1
    Uniform Sprite Size Max=8
    Uniform Sprite Size Min=1.5
    UsePositionOffset=false
    Write Material Random=true
   )
   &SkeletalMeshLocation(Bone Sampling Mode=::NewEnumerator4, MaskSpawnGroup=false, Skeletal Mesh=$System.Net SKM)
   &ShapeLocation(MaskSpawnGroup=false, Non Uniform Scale=<1 0 1>, Override Seed=false, Sphere Radius=5, UseOwnerScale=true)
   &CameraOffset(Camera Offset Amount=&RandomRangeFloat(Minimum=0, Maximum=-15))
   &AddVelocity(Velocity=&FN_FrontEndVector3Modifier(Gameplay=<0 0 0>, Front End=&RandomRangeVector(Minimum=<0 -10 0>, Maximum=<0 -50 0>)))
  ]
  particleUpdateScript [
   &ParticleState()
   &ScaleSpriteSize(
    Non-Uniform Curve Scale=&RandomRangeVector2D(Minimum=<1 1>, Maximum=<1 1.3>)
    Non-Uniform Curve Sprite Scale=DataInterface() {propertyValues {
      XCurve {
       defaultValue=0
       keys [
        {arriveTangent=4.31, interpMode=RCIM_Cubic, leaveTangent=4.31, tangentMode=RCTM_User, time=0, value=0}
        <at 0.8537 0.9403 RCIM_Cubic>
        {arriveTangent=-10.9842, arriveTangentWeight=0.1, interpMode=RCIM_Cubic, leaveTangent=-10.9842, tangentMode=RCTM_User, time=1, value=0}
       ]
      }
      YCurve {keys [
        {arriveTangent=4.31, interpMode=RCIM_Cubic, leaveTangent=4.31, tangentMode=RCTM_User, time=0, value=0}
        <at 0.8537 0.9403 RCIM_Cubic>
        {arriveTangent=-10.9842, interpMode=RCIM_Cubic, leaveTangent=-10.9842, tangentMode=RCTM_User, time=1, value=0}
       ]}
     }}
    Scale Sprite Size Mode=::NewEnumerator3
   )
   &DynamicMaterialParameters(
    Index 0 Param 1=&FloatFromCurve(FloatCurve=DataInterface() {propertyValues {Curve {keys [
         {arriveTangent=0.1706, interpMode=RCIM_Cubic, leaveTangent=0.1706, leaveTangentWeight=0.3081, tangentMode=RCTM_User, time=0, value=0}
         {arriveTangentWeight=0.3104, interpMode=RCIM_Cubic, leaveTangentWeight=0.0345, tangentMode=RCTM_User, time=0.9, value=0.2}
         {arriveTangent=3.2005, interpMode=RCIM_Cubic, leaveTangent=3.2005, tangentMode=RCTM_User, time=1, value=0.5}
        ]}}})
    Module.Index0Float0Write=true
    Module.Index0Float1Write=true
    Module.Index0Float2Write=true
    Module.Index0Float3Write=true
   )
   &ScaleColor(
    Scale Alpha=&FloatFromCurve(
     FloatCurve=DataInterface() {propertyValues {Curve {keys [
         <at 0.0545 0 RCIM_Cubic>
         {interpMode=RCIM_Cubic, leaveTangentWeight=0.3713, tangentMode=RCTM_User, time=0.5, value=1}
         {
          arriveTangent=-3.2389
          arriveTangentWeight=0.386
          interpMode=RCIM_Cubic
          leaveTangent=-3.2389
          leaveTangentWeight=0.1013
          tangentMode=RCTM_User
          time=0.8949
          value=0.5475
         }
         {arriveTangent=-1.4392, interpMode=RCIM_Cubic, leaveTangent=-1.4392, tangentMode=RCTM_User, time=1.0011, value=0.002}
        ]}}}
     Scale Curve=&Multiply_Float(A=1, B=$System.Bubbles TOD)
    )
    ScaleA=true
    ScaleRGB=true
   )
   &PointForce(
    Force Origin=&AddVectorToPosition(Position=$Engine.Emitter.SimulationPosition, Vector=<0 -20 40>)
    Force Strength=25
    Module.Use Falloff=false
    Module.UseLinearForce=false
   )
   /SetVariables(Particles.Position=&AddVectorToPosition(
     Position=$Particles.Position
     Vector=&Divide_Vector(A:Vector3f=$Particles.SkeletalMeshLocation.SampledVelocity, B:Vector3f=<650 900 900>)
    )) {isSetParameters=true}
   &CurlNoiseForce(
    Noise Frequency=75
    Noise Strength=&FloatFromCurve(
     FloatCurve=DataInterface() {propertyValues {Curve {keys [
         {arriveTangent=-39.1213, interpMode=RCIM_Cubic, leaveTangent=-39.1213, leaveTangentWeight=114.1625, tangentMode=RCTM_User, time=0, value=50}
         {arriveTangentWeight=0.3333, interpMode=RCIM_Cubic, leaveTangentWeight=53.6425, tangentMode=RCTM_User, time=1, value=25}
        ]}}}
     Scale Curve=&RandomRangeFloat(Minimum=1, Maximum=3)
    )
    Pan Noise=false
   )
   &GravityForce(Gravity=&MakeVector(
     X=0
     Y=0
     Z=&FloatFromCurve(
      FloatCurve=DataInterface() {propertyValues {Curve {
         defaultValue=1
         keys [
          {interpMode=RCIM_Cubic, tangentMode=RCTM_User, time=0.3826, value=0}
          {arriveTangent=2.6812, interpMode=RCIM_Cubic, leaveTangent=2.6812, tangentMode=RCTM_User, time=0.8533, value=0.484}
          {interpMode=RCIM_Cubic, tangentMode=RCTM_User, time=0.9465, value=1.0229}
         ]
        }}}
      Scale Curve=300
     )
    ))
   &Drag(
    Drag=&FN_FrontEndFloatModifier(Gameplay=0, Front End=&RandomRangeFloat(Minimum=0.25, Maximum=1.5))
    Use Linear Drag=true
    Use Rotational Drag=true
   )
   &SolveForcesAndVelocity(Clamp Velocity=false, Limit Acceleration=false)
   &BlendSpace("Position Blend (World <-> Local)":NiagaraFloat=&FN_FrontEndFloatModifier(
     Gameplay=&MakeCustomFloatFromBool(
      Bool=&SetBoolByFloatComparison(A=&VectorLength(Vector=$User.OwnerLocalVelocity), B=$Emitter.WalkForward Speed)
      True Float=1
      False Float=&Multiply_Float(
       A=1
       B=&FloatFromCurve(FloatCurve=DataInterface() {propertyValues {Curve {keys [<at 0.3 1>,<at 0.8 0>]}}})
      )
     )
     Front End=1
     Translation Contribution:NiagaraFloat=1
     Rotation Contribution:NiagaraFloat=1
    ))
  ]
  rendererClasses [&NiagaraSpriteRendererProperties]
  renderers [{rendererClass=&NiagaraSpriteRendererProperties}]
 }}
summary {
 systemName=NS_Emote_CountyFrogLoaf_NetBubbles
 emitters [{emitterName=BigJelly_Bubbles, rendererClasses [&NiagaraSpriteRendererProperties]}]
 userVariables [
  User.bIsFrontEnd {type=NiagaraBool}
  User.bIsFrontEndPreview {type=NiagaraBool}
  User.OwnerLocalVelocity {type=Vector3f, description="", defaultValue {struct=Vector3f, value=<0 0 0>}}
 ]
}
