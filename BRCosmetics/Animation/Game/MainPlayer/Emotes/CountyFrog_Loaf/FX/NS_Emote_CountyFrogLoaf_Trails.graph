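# Niagara system for the CountyFrog Loaf emote that draws short-lived, dark motion-trail ribbons from a skeletal mesh.
# A single Motion_Trails emitter combines an instantaneous burst with a steady spawn rate, placing particles via SkeletalMeshLocation on the System.Net SKM mesh.
# Spawn color and alpha scale vary by time of day, with randomized shade and opacity per particle.
# Curves taper ribbon width and fade alpha over each particle's life, with drag applied.
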NiagaraSystem /BRCosmetics/Animation/Game/MainPlayer/Emotes/CountyFrog_Loaf/FX/NS_Emote_CountyFrogLoaf_Trails
dependencies {
 usedDataInterfaces [&NiagaraDataInterfaceCurve, &NiagaraDataInterfaceSkeletalMesh]
 usedDynamicInputs []
 usedModules [
  &self:SystemSpawnScript.SystemScriptSource.SystemScriptGraph.NiagaraNodeAssignment_1.SetVariables_#1
  &SystemState
 ]
}
emitters {Motion_Trails {
  emitterUpdateScript [
   &EmitterState(Life Cycle Mode=::NewEnumerator1, Loop Behavior=::NewEnumerator1, Loop Duration=4, UseLoopDelay=false)
   &SpawnBurst_Instantaneous(Spawn Count=15, Use Spawn Probability=false, UseLoopCountLimit=false)
   &SpawnRate(SpawnRate=50, Use Spawn Probability=false)
  ]
  particleSpawnScript [
   &InitializeParticle(
    Color=&FNTimeOfDayColor(
     Night Color=&RandomRangeLinearColor(Minimum=<rgba 0.024 0.024 0.024 0.4>, Maximum=<rgba 0 0 0 0.55>)
     Morning Color=&RandomRangeLinearColor(Minimum=<rgba 0.0365 0.0365 0.0365 0.4>, Maximum=<rgba 0.024 0.024 0.024 0.4>)
     Day Color=&RandomRangeLinearColor(Minimum=<rgba 0.0656 0.0656 0.0656 0.7>, Maximum=<rgba 0.024 0.024 0.024 0.3>)
     Evening Color=&RandomRangeLinearColor(Minimum=<rgba 0.024 0.024 0.024 0.4>, Maximum=<rgba 0.0005 0.0005 0.0073 0.55>)
    )
    Lifetime=0.5
    Ribbon Width=20
    Ribbon Width Mode=::NewEnumerator1
    UsePositionOffset=false
    Write Material Random=true
   )
   &SkeletalMeshLocation(Bone Sampling Mode=::NewEnumerator4, MaskSpawnGroup=false, Skeletal Mesh=$System.Net SKM)
  ]
  particleUpdateScript [
   &ParticleState()
   &ScaleRibbonWidth(Ribbon Width Scale=&FloatFromCurve(FloatCurve=DataInterface() {propertyValues {Curve {
        defaultValue=1
        keys [
         {interpMode=RCIM_Cubic, tangentMode=RCTM_User, time=0, value=0}
         {arriveTangent=0.1693, interpMode=RCIM_Cubic, leaveTangent=0.1693, tangentMode=RCTM_User, time=0.2, value=1}
         {interpMode=RCIM_Cubic, tangentMode=RCTM_User, time=0.5, value=1}
         {tangentMode=RCTM_User, time=1, value=0}
        ]
       }}}))
   &ScaleColor(
    Scale Alpha=&FloatFromCurve(
     FloatCurve=DataInterface() {propertyValues {Curve {
        defaultValue=1
        keys [{interpMode=RCIM_Cubic, tangentMode=RCTM_User, time=0, value=1}, {interpMode=RCIM_Cubic, tangentMode=RCTM_User, time=0.85, value=0}]
       }}}
     Scale Curve=&FNTimeOfDayScalar(Night Scale=7, Morning Scale=10, Day Scale=10, Evening Scale=15)
    )
    ScaleA=true
    ScaleRGB=true
   )
   &Drag(Drag=5, Use Linear Drag=true, Use Rotational Drag=true)
   &SolveForcesAndVelocity(Clamp Velocity=false, Limit Acceleration=false)
  ]
  rendererClasses [&NiagaraRibbonRendererProperties]
  renderers [{rendererClass=&NiagaraRibbonRendererProperties}]
 }}
summary {
 systemName=NS_Emote_CountyFrogLoaf_Trails
 emitters [{emitterName=Motion_Trails, rendererClasses [&NiagaraRibbonRendererProperties]}]
 userVariables []
}
